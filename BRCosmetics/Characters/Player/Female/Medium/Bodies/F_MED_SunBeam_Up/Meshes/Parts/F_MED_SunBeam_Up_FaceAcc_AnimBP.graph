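# Post-style AnimBP for the SunBeam female body's face accessory mesh, adding secondary motion to hair, bun, bangs and earrings.
# It copies the attached parent's pose, then runs serial AnimDynamics bones and chains in root-relative space with angular springs, gravity and damped component motion.
# Constraint nodes after the hair chains drive two right-side master bones from simulated strand rotations, one averaging strands A and B.
# The event graph has no logic.
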AnimBlueprint /BRCosmetics/Characters/Player/Female/Medium/Bodies/F_MED_SunBeam_Up/Meshes/Parts/F_MED_SunBeam_Up_FaceAcc_AnimBP
target_skeleton=&/F_MED_SunBeam_Up_FaceAcc_Skeleton
graphs [
 AnimGraph {nodes [
   n1 = Anim Dynamics - Bone: dyn_bun(ComponentPose=Local To Component(LocalPose=Copy Pose From Mesh() {bUseAttachedParent=true}), Alpha=0.7, ExternalForce="0,0,0") {
    angularDampingOverride=1
    angularSpringConstant=20000
    bAngularSpring=true
    bOverrideAngularDamping=true
    bUseGravityOverride=true
    simulationSpace=RootRelative
    alphaBoolBlend {blendOption=Linear}
    boundBone {boneName=dyn_bun}
    componentAppliedLinearAccClamp=<2000 2000 2000>
    componentLinearAccScale=<0.3 0.3 0.3>
    componentLinearVelScale=<0.3 0.3 0.3>
    gravityOverride {z=-980}
    physicsBodyDefinitions [{
      boundBone {boneName=dyn_bun}
      boxExtents=<10 20 20>
      constraintSetup {angularLimitsMax=<5 5 5>, angularLimitsMin=<-5 -5 -5>, angularTarget=<0 0 0>}
      localJointOffset=<4 0 0>
     }]
   }
   n2 = Anim Dynamics - Bone: dyn_bunHair_01(ComponentPose=@n1, Alpha=0.7, ExternalForce="0,0,0") {
    angularDampingOverride=1
    angularSpringConstant=20000
    bAngularSpring=true
    bOverrideAngularDamping=true
    bUseGravityOverride=true
    simulationSpace=RootRelative
    alphaBoolBlend {blendOption=Linear}
    boundBone {boneName=dyn_bunHair_01}
    componentAppliedLinearAccClamp=<2000 2000 2000>
    componentLinearAccScale=<0.3 0.3 0.3>
    componentLinearVelScale=<0.3 0.3 0.3>
    gravityOverride {z=-980}
    physicsBodyDefinitions [{
      boundBone {boneName=dyn_bunHair_01}
      boxExtents=<10 20 20>
      constraintSetup {angularLimitsMax=<1 0 15>, angularLimitsMin=<-1 -15 -15>, angularTarget=<0 0 0>}
      localJointOffset=<4 0 0>
     }]
   }
   n3 = Anim Dynamics - Bone: dyn_bunHair_02(ComponentPose=@n2, Alpha=0.7, ExternalForce="0,0,0") {
    angularDampingOverride=1
    angularSpringConstant=20000
    bAngularSpring=true
    bOverrideAngularDamping=true
    bUseGravityOverride=true
    simulationSpace=RootRelative
    alphaBoolBlend {blendOption=Linear}
    boundBone {boneName=dyn_bunHair_02}
    componentAppliedLinearAccClamp=<2000 2000 2000>
    componentLinearAccScale=<0.3 0.3 0.3>
    componentLinearVelScale=<0.3 0.3 0.3>
    gravityOverride {z=-980}
    physicsBodyDefinitions [{
      boundBone {boneName=dyn_bunHair_02}
      boxExtents=<10 20 20>
      constraintSetup {angularLimitsMax=<1 0 15>, angularLimitsMin=<-1 -15 -15>, angularTarget=<0 0 0>}
      localJointOffset=<4 0 0>
     }]
   }
   n4 = Anim Dynamics - Bone: dyn_bunHair_03(ComponentPose=@n3, Alpha=0.7, ExternalForce="0,0,0") {
    angularDampingOverride=1
    angularSpringConstant=20000
    bAngularSpring=true
    bOverrideAngularDamping=true
    bUseGravityOverride=true
    simulationSpace=RootRelative
    alphaBoolBlend {blendOption=Linear}
    boundBone {boneName=dyn_bunHair_03}
    componentAppliedLinearAccClamp=<2000 2000 2000>
    componentLinearAccScale=<0.3 0.3 0.3>
    componentLinearVelScale=<0.3 0.3 0.3>
    gravityOverride {z=-980}
    physicsBodyDefinitions [{
      boundBone {boneName=dyn_bunHair_03}
      boxExtents=<10 20 20>
      constraintSetup {angularLimitsMax=<1 0 15>, angularLimitsMin=<-1 -15 -15>, angularTarget=<0 0 0>}
      localJointOffset=<4 0 0>
     }]
   }
   n5 = Anim Dynamics - Bone: dyn_earring_l(ComponentPose=@n4, Alpha=0.9, ExternalForce="0,0,0") {
    angularDampingOverride=0.97
    angularSpringConstant=20000
    bAngularSpring=true
    bOverrideAngularDamping=true
    bUseGravityOverride=true
    simulationSpace=RootRelative
    alphaBoolBlend {blendOption=Linear}
    boundBone {boneName=dyn_earring_l}
    componentAppliedLinearAccClamp=<2000 2000 2000>
    componentLinearAccScale=<0.3 0.3 0.3>
    componentLinearVelScale=<0.3 0.3 0.3>
    gravityOverride {z=-980}
    physicsBodyDefinitions [{
      boundBone {boneName=dyn_earring_l}
      boxExtents=<10 20 20>
      constraintSetup {angularLimitsMax=<45 45 45>, angularLimitsMin=<-45 -45 -45>, angularTarget=<0 0 0>}
      localJointOffset=<4 0 0>
     }]
   }
   n6 = Anim Dynamics - Bone: dyn_earring_r(ComponentPose=@n5, Alpha=1, ExternalForce="0,0,0") {
    angularDampingOverride=0.97
    angularSpringConstant=20000
    bAngularSpring=true
    bOverrideAngularDamping=true
    bUseGravityOverride=true
    simulationSpace=RootRelative
    alphaBoolBlend {blendOption=Linear}
    boundBone {boneName=dyn_earring_r}
    componentAppliedLinearAccClamp=<2000 2000 2000>
    componentLinearAccScale=<0.3 0.3 0.3>
    componentLinearVelScale=<0.3 0.3 0.3>
    gravityOverride {z=-980}
    physicsBodyDefinitions [{
      boundBone {boneName=dyn_earring_r}
      boxExtents=<10 20 20>
      constraintSetup {angularLimitsMax=<45 45 45>, angularLimitsMin=<-45 -45 -45>, angularTarget=<0 0 0>}
      localJointOffset=<4 0 0>
     }]
   }
   n7 = "Anim Dynamics - Chain: dyn_hair_A_01_l -> dyn_hair_A_02_l"(ComponentPose=@n6, Alpha=1, ExternalForce="0,0,0") {
    angularDampingOverride=1
    angularSpringConstant=20000
    bAngularSpring=true
    bChain=true
    bOverrideAngularDamping=true
    bUseGravityOverride=true
    simulationSpace=RootRelative
    alphaBoolBlend {blendOption=Linear}
    boundBone {boneName=dyn_hair_A_01_l}
    chainEnd {boneName=dyn_hair_A_02_l}
    componentAppliedLinearAccClamp=<2000 2000 2000>
    componentLinearAccScale=<0.3 0.3 0.3>
    componentLinearVelScale=<0.3 0.3 0.3>
    gravityOverride {z=-980}
    physicsBodyDefinitions [
     {
      boundBone {boneName=dyn_hair_A_01_l}
      boxExtents=<10 20 20>
      constraintSetup {angularLimitsMax=<1 0 45>, angularLimitsMin=<-1 -45 -45>, angularTarget=<0 0 0>}
      localJointOffset=<4 0 0>
     }
     {
      boundBone {boneName=dyn_hair_A_02_l}
      boxExtents=<10 20 20>
      constraintSetup {angularLimitsMax=<1 0 45>, angularLimitsMin=<-1 -45 -45>, angularTarget=<0 0 0>}
      localJointOffset=<4 0 0>
     }
    ]
   }
   n8 = "Anim Dynamics - Chain: dyn_hair_B_01_l -> dyn_hair_B_02_l"(ComponentPose=@n7, Alpha=1, ExternalForce="0,0,0") {
    angularDampingOverride=1
    angularSpringConstant=20000
    bAngularSpring=true
    bChain=true
    bOverrideAngularDamping=true
    bUseGravityOverride=true
    simulationSpace=RootRelative
    alphaBoolBlend {blendOption=Linear}
    boundBone {boneName=dyn_hair_B_01_l}
    chainEnd {boneName=dyn_hair_B_02_l}
    componentAppliedLinearAccClamp=<2000 2000 2000>
    componentLinearAccScale=<0.3 0.3 0.3>
    componentLinearVelScale=<0.3 0.3 0.3>
    gravityOverride {z=-980}
    physicsBodyDefinitions [
     {
      boundBone {boneName=dyn_hair_B_01_l}
      boxExtents=<10 20 20>
      constraintSetup {angularLimitsMax=<1 0 45>, angularLimitsMin=<-1 -45 -15>, angularTarget=<0 0 0>}
      localJointOffset=<4 0 0>
     }
     {
      boundBone {boneName=dyn_hair_B_02_l}
      boxExtents=<10 20 20>
      constraintSetup {angularLimitsMax=<1 0 45>, angularLimitsMin=<-1 -45 -15>, angularTarget=<0 0 0>}
      localJointOffset=<4 0 0>
     }
    ]
   }
   n9 = "Anim Dynamics - Chain: dyn_hair_A_01_r -> dyn_hair_A_03_r"(ComponentPose=@n8, Alpha=1, ExternalForce="0,0,0") {
    angularDampingOverride=1
    angularSpringConstant=20000
    bAngularSpring=true
    bChain=true
    bOverrideAngularDamping=true
    bUseGravityOverride=true
    simulationSpace=RootRelative
    alphaBoolBlend {blendOption=Linear}
    boundBone {boneName=dyn_hair_A_01_r}
    chainEnd {boneName=dyn_hair_A_03_r}
    componentAppliedLinearAccClamp=<2000 2000 2000>
    componentLinearAccScale=<0.3 0.3 0.3>
    componentLinearVelScale=<0.3 0.3 0.3>
    gravityOverride {z=-980}
    physicsBodyDefinitions [
     {
      boundBone {boneName=dyn_hair_A_01_r}
      boxExtents=<10 20 20>
      constraintSetup {angularLimitsMax=<1 0 30>, angularLimitsMin=<-1 -45 -30>, angularTarget=<0 0 0>}
      localJointOffset=<4 0 0>
     }
     {
      boundBone {boneName=dyn_hair_A_02_r}
      boxExtents=<10 20 20>
      constraintSetup {angularLimitsMax=<1 0 30>, angularLimitsMin=<-1 -45 -30>, angularTarget=<0 0 0>}
      localJointOffset=<4 0 0>
     }
     {
      boundBone {boneName=dyn_hair_A_03_r}
      boxExtents=<10 20 20>
      constraintSetup {angularLimitsMax=<1 0 30>, angularLimitsMin=<-1 -45 -30>, angularTarget=<0 0 0>}
      localJointOffset=<4 0 0>
     }
    ]
   }
   n10 = "Anim Dynamics - Chain: dyn_hair_B_01_r -> dyn_hair_B_02_r"(ComponentPose=@n9, Alpha=1, ExternalForce="0,0,0") {
    angularDampingOverride=1
    angularSpringConstant=20000
    bAngularSpring=true
    bChain=true
    bOverrideAngularDamping=true
    bUseGravityOverride=true
    simulationSpace=RootRelative
    alphaBoolBlend {blendOption=Linear}
    boundBone {boneName=dyn_hair_B_01_r}
    chainEnd {boneName=dyn_hair_B_02_r}
    componentAppliedLinearAccClamp=<2000 2000 2000>
    componentLinearAccScale=<0.3 0.3 0.3>
    componentLinearVelScale=<0.3 0.3 0.3>
    gravityOverride {z=-980}
    physicsBodyDefinitions [
     {
      boundBone {boneName=dyn_hair_B_01_r}
      boxExtents=<10 20 20>
      constraintSetup {angularLimitsMax=<45 0 45>, angularLimitsMin=<-45 -45 -45>, angularTarget=<0 0 0>}
      localJointOffset=<4 0 0>
     }
     {
      boundBone {boneName=dyn_hair_B_02_r}
      boxExtents=<10 20 20>
      constraintSetup {angularLimitsMax=<45 0 45>, angularLimitsMin=<-45 -45 -45>, angularTarget=<0 0 0>}
      localJointOffset=<4 0 0>
     }
    ]
   }
   n11 = "Constraint - dyn_hair_A2_00_r_MSTR (RR)"!(ComponentPose=@n10, Alpha=1, ConstraintWeights_0=0.5, ConstraintWeights_1:float=0.5) -> (Pose) {
    alphaBoolBlend {blendOption=Linear}
    boneToModify {boneName=dyn_hair_A2_00_r_MSTR}
    constraintSetup [{transformType=Rotation, targetBone {boneName=dyn_hair_A_01_r}}, {transformType=Rotation, targetBone {boneName=dyn_hair_B_01_r}}]
    constraintWeights [1,1]
   }
   n12 = "Constraint - dyn_hair_B2_00_r_MSTR (RT)"!(ComponentPose=@n11, Alpha=1, ConstraintWeights_0=1, ConstraintWeights_1:float=0) -> (Pose) {
    alphaBoolBlend {blendOption=Linear}
    boneToModify {boneName=dyn_hair_B2_00_r_MSTR}
    constraintSetup [{transformType=Rotation, targetBone {boneName=dyn_hair_B_01_r}}, {transformType=Translation, targetBone {boneName=None}}]
    constraintWeights [1,1]
   }
   n13 = Anim Dynamics - Bone: dyn_bangs_r(ComponentPose=@n12, Alpha=1, ExternalForce="0,0,0") {
    angularDampingOverride=0.97
    angularSpringConstant=20000
    bAngularSpring=true
    bOverrideAngularDamping=true
    bUseGravityOverride=true
    simulationSpace=RootRelative
    alphaBoolBlend {blendOption=Linear}
    boundBone {boneName=dyn_bangs_r}
    componentAppliedLinearAccClamp=<2000 2000 2000>
    componentLinearAccScale=<0.3 0.3 0.3>
    componentLinearVelScale=<0.3 0.3 0.3>
    gravityOverride {z=-980}
    physicsBodyDefinitions [{
      boundBone {boneName=dyn_bangs_r}
      boxExtents=<10 20 20>
      constraintSetup {angularLimitsMax=<1 0 10>, angularLimitsMin=<-1 -45 -10>, angularTarget=<0 0 0>}
      localJointOffset=<4 0 0>
     }]
   }
   n14 = Anim Dynamics - Bone: dyn_bangs_c(ComponentPose=@n13, Alpha=1, ExternalForce="0,0,0") {
    angularDampingOverride=1
    angularSpringConstant=20000
    bAngularSpring=true
    bOverrideAngularDamping=true
    bUseGravityOverride=true
    simulationSpace=RootRelative
    alphaBoolBlend {blendOption=Linear}
    boundBone {boneName=dyn_bangs_c}
    componentAppliedLinearAccClamp=<2000 2000 2000>
    componentLinearAccScale=<0.3 0.3 0.3>
    componentLinearVelScale=<0.3 0.3 0.3>
    gravityOverride {z=-980}
    physicsBodyDefinitions [{
      boundBone {boneName=dyn_bangs_c}
      boxExtents=<10 20 20>
      constraintSetup {angularLimitsMax=<1 0 10>, angularLimitsMin=<-1 -45 -10>, angularTarget=<0 0 0>}
      localJointOffset=<4 0 0>
     }]
   }
   n15 = Anim Dynamics - Bone: dyn_bangs_l(ComponentPose=@n14, Alpha=0.9, ExternalForce="0,0,0") {
    angularDampingOverride=1
    angularSpringConstant=20000
    bAngularSpring=true
    bOverrideAngularDamping=true
    bUseGravityOverride=true
    simulationSpace=RootRelative
    alphaBoolBlend {blendOption=Linear}
    boundBone {boneName=dyn_bangs_l}
    componentAppliedLinearAccClamp=<2000 2000 2000>
    componentLinearAccScale=<0.3 0.3 0.3>
    componentLinearVelScale=<0.3 0.3 0.3>
    gravityOverride {z=-980}
    physicsBodyDefinitions [{
      boundBone {boneName=dyn_bangs_l}
      boxExtents=<10 20 20>
      constraintSetup {angularLimitsMax=<1 0 10>, angularLimitsMin=<-1 -45 -10>, angularTarget=<0 0 0>}
      localJointOffset=<4 0 0>
     }]
   }
   Output Pose(Result=Component To Local(ComponentPose=@n15)) {layerGroup=None}
  ]}
 EventGraph {nodes [Event BlueprintUpdateAnimation()]}
]
variables []
